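# Environment query for the VividRazor minion Warp ability that picks a teleport spot a short way ahead, toward the NPC's target.
# An OnCircle generator places a few points on a narrow arc around the querier, aimed at the Warp target context, with tracing off.
# Points project onto navmesh using the Phoebe no-smashable, no-water filter.
# A pathfinding test with the same filter evaluates each point against the target.
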EnvQuery /VividRazor_Minions/Abilities/EQS/Warp_EQS_TeleportSlightlyForward
options [{
  generator {
   name=EnvQueryGenerator_OnCircle_0
   class=EnvQueryGenerator_OnCircle
   properties {
    arcAngle=60
    bDefineArc=true
    circleRadius=1000
    numberOfPoints=5
    arcDirection {lineTo=/NPCLibrary/Abilities/Warp/EQS/EnvQueryContext_WarpGetNPCTarget.EnvQueryContext_WarpGetNPCTarget_C}
    projectionData {
     extentX=0
     navigationFilter=&NavFilter_Phoebe_NoSmashableNoWater_C
     postProjectionVerticalOffset=0
     projectDown=150
     projectUp=280
    }
    traceData {navigationFilter=null, traceMode=None}
   }
  }
  tests [EnvQueryTest_Pathfinding_2 {
    class=EnvQueryTest_Pathfinding
    properties {
     context=&EnvQueryContext_WarpGetNPCTarget_C
     filterClass=&NavFilter_Phoebe_NoSmashableNoWater_C
    }
   }]
 }]
